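# Nanite-enabled lit material for the NeoTilted spotlight prop, which sways its red-vertex-color-masked vertices through World Position Offset and adds tinted, masked emissive.
# A switchable sine rotation swings them around a local axis at an offset pivot from the actor position.
# Normal correction uses its own axis and speed, varied by object scale, so it differs from the vertex motion.
# All attributes pass through MF_CreativeMoveTool.
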Material /FigmentCore/Athena/Effects/Materials/NeoTilted/M_NeoTilted_SpotLight_WPO
graph {
 properties {bUseMaterialAttributes=true, bUsedWithNanite=true, numCustomizedUVs=5, pixelDepthOffsetMode=PDOM_Legacy, usedShadingModels=Default Lit}
 nodes [
  AppendVector(A=TextureCoordinate() {coordinateIndex=3}, B=ComponentMask(None=TextureCoordinate() {coordinateIndex=4}) {r=true})
  n1 = TextureSample() {samplerType=SAMPLERTYPE_Masks, texture=/Game/Environments/Sets/Neo/Props/Textures/T_NeoTilted_Spotlight_01_S}
  n2 = VectorParameter() {parameterName=Emissive Color Multiplier, defaultValue=<rgba 0 4.9533 10 0>}
  n3 = Constant3Vector() {constant=<rgba 0 0 1 0>}
  n4 = Multiply(
   A=Constant() {r=34.9111}
   B=Frac(None=MaterialFunctionCall() {materialFunction=/Engine/Functions/Engine_MaterialFunctions02/WorldPositionOffset/ObjectScale}.Scale X)
  ) {constA=16}
  n5 = Constant3Vector()
  n6 = VectorParameter() {parameterName=Rotation Vector, defaultValue=<rgba 1 0 0 1>}
  n7 = Multiply(
   A=Time()
   B=LinearInterpolate(
    A=ScalarParameter() {defaultValue=0.125, parameterName=Rotation Speed}
    B=ScalarParameter() {defaultValue=0.07, parameterName=Rotation Speed2}
    Alpha=Frac(None=@n4)
   )
  )
  n8 = MaterialFunctionCall(
   Rotation Axis=Normalize(VectorInput=RotateAboutAxis(NormalizedRotationAxis=@n3, RotationAngle=@n4, PivotPoint=@n5, Position=@n6.RGB))
   Rotation Angle=Multiply(A=Sine(None=@n7), B=ScalarParameter() {defaultValue=0.05, parameterName=Rotation Max})
  ) {materialFunction=&FixRotateAboutAxisNormals}
  n9 = VertexInterpolator(VS=@n8.New Tangent Space Vertex Normal)
  n10 = TextureSample() {samplerType=SAMPLERTYPE_Normal, texture=/Game/Environments/Sets/Neo/Props/Textures/T_NeoTilted_Spotlight_01_N}
  n11 = VectorParameter() {parameterName=LocalAxis, defaultValue=<rgba 1 0 0 1>}
  n12 = Sine(None=Multiply(A=Time(), B=ScalarParameter() {defaultValue=0.05, parameterName=Speed}))
  n13 = ScalarParameter() {defaultValue=0.05, parameterName=Rotation Max}
  n14 = ActorPositionWS()
  n15 = Transform(None=VectorParameter() {parameterName=Pivot Offset Custom, defaultValue=<rgba 0 0 75 1>}.RGB) {transformSourceType=TRANSFORMSOURCE_Local}
  n16 = StaticSwitchParameter(
   True=RotateAboutAxis(
    NormalizedRotationAxis=Transform(None=@n11.RGB) {transformSourceType=TRANSFORMSOURCE_Local}
    RotationAngle=Multiply(A=@n12, B=@n13)
    PivotPoint=Add(A=@n14, B=@n15)
    Position=WorldPosition().XYZ
   )
   False=Constant()
  ) {defaultValue=true, parameterName=Rotate}
  n17 = VertexColor()
  MaterialFunctionCall(MA In=MakeMaterialAttributes!(
    BaseColor=TextureSample() {texture=/Game/Environments/Sets/Neo/Props/Textures/T_NeoTilted_Spotlight_01_D}.RGB
    Metallic=@n1.G
    Specular=@n1.G
    Roughness=@n1.G
    Anisotropy
    EmissiveColor=Multiply(A=@n1.A, B=@n2.RGB)
    Opacity
    OpacityMask
    Normal=MaterialFunctionCall(BaseNormal=@n9.PS, AdditionalNormal=@n10.RGB) {materialFunction=&BlendAngleCorrectedNormals}.Result
    Tangent
    WorldPositionOffset=Multiply(A=@n16, B=@n17.R)
    SubsurfaceColor
    ClearCoat
    ClearCoatRoughness
    AmbientOcclusion
    Refraction
    CustomizedUVs_0
    CustomizedUVs_1
    CustomizedUVs_2
    CustomizedUVs_3=@n8.New Tangent Space Vertex Normal RG
    CustomizedUVs_4=@n8.New Tangent Space Vertex Normal RG
    CustomizedUVs_5
    CustomizedUVs_6
    CustomizedUVs_7
    PixelDepthOffset
    ShadingModel
    Displacement
   )) {materialFunction=&MF_CreativeMoveTool}
 ]
}
